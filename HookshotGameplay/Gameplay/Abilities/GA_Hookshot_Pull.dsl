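; GA_Hookshot_Pull is a locally predicted knockback-derived gameplay ability that pulls the hooked player or AI pawn toward the hookshot's instigator.
; CanActivateAbility gates pulls through hotfixable team, DBNO, flight and movement-mode rules.
; Activation ends special movement modes, then the parent launches using an arc velocity clamped by pull distance.
; A hotfix switches between an MME pull timed by distance over speed and a relaunch when the parent's launch finishes.
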(event Ability|EventActivateAbility
  ;[0.01927 0.1042 0.02651] Event ActivateAbility: Run logic to launch the hooked (target) player to the hooking (instigator) player.
  ; This ability is only ended through the GA when it fails. If the ability is successful, it is ended when the gameplay effect that added it expires at the end of the effect duration.
  ; The gameplay ability is locally predicted so this logic is run on both server and client to maintain parity as reduce correction.
  ;;[0.3297 0.06442 0.3594] Cache target and instigator variables.
  (bind _character (General|GetTargetCharacterfromAvatar (Ability|GetAvatarActorfromActorInfo)))
  (bind _output_get (Variables|Transient|SetTargetCharacter _character))
  (Utilities|IsValid _output_get
    (:Is%20Valid
      (bind _playerpawn (General|GetInstigatorPlayerPawn))
      (bind _output_get_2 (Variables|Transient|SetInstigatorPlayerPawn _playerpawn))
      (Utilities|IsValid _output_get_2
        (:Is%20Valid
          ;;[0.1146 0.1133 0.02145] Cache the instigator  pawn location.
          (Variables|Transient|SetInstigatorPlayerPawnLocation (Transformation|GetActorLocation _output_get_2))
          ;;[0.1083 0.7969 0.03735] End custom movement modes.
          (EndMovement|EndMovementModes)
          ;;[1 0.1475 0.9283] Let the parent perform the actual knockback.
          (|CallParentFunction)
          ;;[0.1354 0.08005 0.04688] Camera lens effect for the pulled player.
          (Ability|ExecuteGameplayCueOnOwner (Variables|General|GetHookshotPulledCameraEffectGameplayCue) (Ability|GetGrantedbyEffectContext)))
        (:Is%20Not%20Valid
          _)))
    (:Is%20Not%20Valid
      _))
  ;;[0.1406 0.07176 0.02266] End Ability
  (Ability|EndAbility))

(event EventLaunchFinished
  ;[0.1667 1 1] LaunchFinished: The root motion launch in the parent has finished, apply launch on the character again so that it lands where it is supposed to.
  ; IF bSuccess is FALSE, EndAbility() since the Pull failed.
  ; IF bSuccess is TRUE, do not called end ability as it will be called when the owning GE is removed after the pull duration.
  (|CallParentFunction)
  (bind _bsuccess (Pull|PullTargetCharacter))
  (if _bsuccess
    (Analytics|NotifyAnalyticsSuccessfulHookshot)
    (else
      (Ability|EndAbility))))

(event Ability|EventOnEndAbility
  (params (bWasCancelled :type Boolean))
  ;[0.1667 0.1059 0.03231] EndAbility: Clean up.
  (Utilities|FlowControl|Sequence
    (:then_0
      (bind _asfort_movement_comp_character (Utilities|Casting|CastToFortMovementComp_Character (Variables|/Script/Engine.Character|GetCharacterMovement (Variables|Transient|GetTargetCharacter))))
      (Class|FortMovementCompCharacter|CancelActiveExtensionwithTag _asfort_movement_comp_character (Variables|Default|GetMME%20Extension%20Tag))
      (bind _asynctask_constantforce (Variables|Transient|GetAsyncTask_ConstantForce)
        (:then
          (GameplayTasks|EndTask _asynctask_constantforce))
        (:else)))
    (:then_1
      (|CallParentFunction bWasCancelled))))

(event EventDoLaunch
  (params (Vel :type Vector :default "0, 0, 0"))
  ;[0.15 0 0.1106 0.5] DoLaunch: Overridden to execute MME pull.
  (if (ScalableFloat|GetValueasBool (Variables|hotfix|GetHotfix_UseMMEPull))
    (bind _targetcharacter (Variables|Transient|GetTargetCharacter)
      (:then
        (Utilities|Casting|CastToFortPlayerPawn _targetcharacter
          (:then
            (Utilities|Casting|CastToFortPlayerController (Pawn|GetController _targetcharacter)
              (:then
                (Ability|HasAuthority
                  (:True
                    (GameplayTasks|TaskWaitDelay 0 (MME|GetMMEPullDuration)
                      (:OnFinish
                        (CallFunction|LaunchFinished))))
                  (:False
                    _)))
              (:CastFailed
                (Utilities|Casting|CastToAIController (Pawn|GetController _targetcharacter)
                  (:then
                    _)
                  (:CastFailed
                    (|CallParentFunction Vel)))))
            (CallFunction|ApplyPullMME))
          (:CastFailed
            _)))
      (:else
        _))
    (else
      _)))

(event Custom|ApplyPullMME)

(fn CalculatePullVelocityForTarget
  (params (TargetActor :type Actor%20Object%20Reference))
  (bind _returnvalue (Transformation|GetActorLocation TargetActor))
  (bind _instigatorplayerpawnlocation (Variables|Transient|GetInstigatorPlayerPawnLocation))
  ;[0.09896 0.02913 0.09767] CalculatePullVelocityForTarget: Calculate the velocity with which the target should be pulled towards the instigator so that it lands at the appropriate distance.
  (Utilities|FlowControl|Sequence
    (:then_0
      (Utilities|IsValid TargetActor
        (:Is%20Valid
          (LocalVariables|SetHalfwayLocation (/ (+ _returnvalue _instigatorplayerpawnlocation) 2.0))
          (LocalVariables|SetSuggestedPullLocation (+ _instigatorplayerpawnlocation (* (Math|Vector|Normalize (- _returnvalue _instigatorplayerpawnlocation)) (ScalableFloat|GetValueatLevel (Variables|Hotfix|GetHotfix_MaxPullLocation)))))
          (LocalVariables|SetEndPosition (Math|Vector|SelectVector (LocalVariables|GetHalfwayLocation) (LocalVariables|GetSuggestedPullLocation) (< (Math|Vector|DistanceSquared(Vector) (LocalVariables|GetHalfwayLocation) (Transformation|GetActorLocation (Variables|Transient|GetInstigatorPlayerPawn))) (Math|Vector|DistanceSquared(Vector) (LocalVariables|GetSuggestedPullLocation) (Transformation|GetActorLocation (Variables|Transient|GetInstigatorPlayerPawn))))))
          (LocalVariables|SetStartPosition (Transformation|GetActorLocation (Variables|GetTargetActor)))
          ;;[0.07089 0.1879 0.2135] Clamp the EndLocation to satisfy the max pull distance value.
          (bind _output_get (LocalVariables|SetEndPosition (+ (LocalVariables|GetStartPosition) (* (Math|Float|Min(Float) (Math|Vector|Distance(Vector) (LocalVariables|GetStartPosition) (LocalVariables|GetEndPosition)) (ScalableFloat|GetValueatLevel (Variables|hotfix|GetHotfix_PullMaxDistance))) (Math|Vector|GetUnitDirection(Vector) (LocalVariables|GetStartPosition) (LocalVariables|GetEndPosition))))))
          (bind _outlaunchvelocity (Game|SuggestProjectileVelocityCustomArc (LocalVariables|GetStartPosition) _output_get 0.0 (ScalableFloat|GetValueatLevel (Variables|Hotfix|GetHotfix_PullArc)))) ; Finds out how much force is needed to put the target 300 units in front of where instigator is facing
          (return _outlaunchvelocity))))
    (:then_1
      (return "0, 0, 0"))))

(fn CanInstigatorPullTarget
  (params (InstigatorPlayerPawn :type Fort%20Player%20Pawn%20Object%20Reference) (TargetCharacter :type Character%20Object%20Reference))
  (bind _bonsameteam (LocalVariables|GetbOnSameTeam))
  (bind _hotfixmap_pullsameteamcustommovement (Variables|Hotfix|SameTeam|GetHotfixMap_PullSameTeamCustomMovement))
  (bind _hotfixmap_pullotherteamcustommovement (Variables|Hotfix|OtherTeam|GetHotfixMap_PullOtherTeamCustomMovement))
  (bind _returnvalue (select _bonsameteam _hotfixmap_pullsameteamcustommovement _hotfixmap_pullotherteamcustommovement))
  (bind _bonsameteam_2 (LocalVariables|GetbOnSameTeam))
  (bind _hotfixmap_pullsameteamdefaultmovement (Variables|Hotfix|SameTeam|GetHotfixMap_PullSameTeamDefaultMovement))
  (bind _hotfixmap_pullotherteamdefaultmovement (Variables|Hotfix|OtherTeam|GetHotfixMap_PullOtherTeamDefaultMovement))
  (bind _returnvalue_2 (select _bonsameteam_2 _hotfixmap_pullsameteamdefaultmovement _hotfixmap_pullotherteamdefaultmovement))
  ;[0.151 0.03775 0.1089] CanInstigatorPullTarget: Check to see if the instigator can actually pull the target.
  (Utilities|FlowControl|Sequence
    (:then_0
      (Utilities|IsValid InstigatorPlayerPawn
        (:Is%20Valid
          (Utilities|IsValid TargetCharacter
            (:Is%20Valid
              (bind _returnvalue_3 (GameplayTags|HasAnyMatchingGameplayTags(Message) TargetCharacter (Variables|KnockbackParent|GetTC_TagsToIgnore)))
              (if (not _returnvalue_3)
                (LocalVariables|SetbOnSameTeam (Team|OnSameTeam InstigatorPlayerPawn TargetCharacter))
                ;;[1 0.3128 0.4736] Check to see if pulling target actors is allowed based on team.
                (if (ScalableFloat|GetValueasBool (select (LocalVariables|GetbOnSameTeam) (Variables|Hotfix|SameTeam|GetHotfix_CanPullSameTeam) (Variables|Hotfix|OtherTeam|GetHotfix_CanPullOtherTeam)))
                  (bind _asfort_player_pawn (Utilities|Casting|CastToFortPlayerPawn TargetCharacter)
                    (:then
                      ;;[0.05373 0.1701 0.2188] Handle friendly player in DBNO
                      (if (Health|IsDBNO _asfort_player_pawn)
                        (return (ScalableFloat|GetValueasBool (select (LocalVariables|GetbOnSameTeam) (Variables|Hotfix|SameTeam|GetHotfix_CanPullSameTeamIsDBNO) (Variables|Hotfix|OtherTeam|GetHotfix_CanPullOtherTeamIsDBNO))))
                        (else
                          ;;[0.2708 0.1166 0.07035] Handle winged flight.
                          (if (GameplayTags|HasAnyMatchingGameplayTags _asfort_player_pawn (GameplayTags|MakeLiteralGameplayTagContainer "(GameplayTags=((TagName=\"Gameplay.Athena.ActivePower.Leap\"),(TagName=\"MovementMode.Flying.WingedFlight\")))"))
                            (return (ScalableFloat|GetValueasBool (select (LocalVariables|GetbOnSameTeam) (Variables|Hotfix|SameTeam|GetHotfix_CanPullSameTeamIsFlying) (Variables|Hotfix|OtherTeam|GetHotfix_CanPullOtherTeamIsFlying))))
                            (else
                              ;;[0.01945 0.0479 0.0625] Handle movement modes.
                              ;;;[0.1198 0.05291 0.06248] Handle custom movement modes using tags first.
                              (bind _gameplaytags (General|GetCustomMovementModeGameplayTags))
                              (for _array_element _gameplaytags
                                (if (GameplayTags|HasMatchingGameplayTag _asfort_player_pawn _array_element)
                                  (if (.ReturnValue (Utilities|Map|Find _returnvalue _array_element))
                                    (return (ScalableFloat|GetValueasBool (Utilities|Map|Find _returnvalue _array_element))))))
                              ;;;[0.003529 0.02688 0.03125] If no custom movement modes are found, use the default movement enums.
                              (bind _charactermovement (Variables|/Script/Engine.Character|GetCharacterMovement _asfort_player_pawn)
                                (:then
                                  (if (.ReturnValue (Utilities|Map|Find _returnvalue_2 (Variables|/Script/Engine.CharacterMovementComponent|GetMovementMode _charactermovement)))
                                    (return (ScalableFloat|GetValueasBool (Utilities|Map|Find _returnvalue_2 (Variables|/Script/Engine.CharacterMovementComponent|GetMovementMode _charactermovement))))
                                    (else
                                      (return (ScalableFloat|GetValueasBool (select (LocalVariables|GetbOnSameTeam) (Variables|Hotfix|SameTeam|GetHotfix_CanPullSameTeamMiscStates) (Variables|Hotfix|OtherTeam|GetHotfix_CanPullOtherTeamMiscStates)))))))
                                (:else)))))))
                    (:CastFailed
                      ;;[0.07147 0.1351 0.1719] Handle non-player pawns.
                      (return (ScalableFloat|GetValueasBool (select (LocalVariables|GetbOnSameTeam) (Variables|Hotfix|SameTeam|GetHotfix_CanPullSameTeamNonPlayerPawns) (Variables|Hotfix|OtherTeam|GetHotfix_CanPullOtherTeamNonPlayerPawns)))))))))))))
    (:then_1
      (return false))))

(fn EndMovement_Ziplining
  (params (PlayerPawn :type Fort%20Player%20Pawn%20Object%20Reference))
  ;[0.00748 0.1198 0.1186] EndMovement_Ziplining: End the zipline movement on the player.
  (Utilities|IsValid PlayerPawn
    (:Is%20Valid
      (Ability|HasAuthority
        (:True
          (bind _returnvalue (Zipline|GetActiveZipline PlayerPawn))
          (Utilities|IsValid _returnvalue
            (:Is%20Valid
              (Class|FortAthenaZiplineBase|AuthorityForceEndZiplining _returnvalue PlayerPawn))))
        (:False
          (Zipline|EndZiplining PlayerPawn false false))))))

(fn EndMovementModes
  (bind _asfort_player_pawn (Utilities|Casting|CastToFortPlayerPawn (Variables|Transient|GetTargetCharacter)))
  (bind _output_get (LocalVariables|SetPlayerPawn _asfort_player_pawn))
  (Ability|SendGameplayEventtoActor _output_get "(TagName=\"Event.Athena.Balloons.RemoveAll\")") ; Remove any balloons on the player.
  ;[0.06532 0.01858 0.1302] End vehicle driving and passenger movement modes.
  (if (GameplayTags|HasAnyTags (GameplayTags|GetOwnedGameplayTags _output_get :!class /Script/GameplayTags.GameplayTagAssetInterface) (GameplayTags|MakeLiteralGameplayTagContainer "(GameplayTags=((TagName=\"MovementMode.Driving\"),(TagName=\"MovementMode.Passenger\")))") true)
    (bind _playerpawn (General|GetInstigatorPlayerPawn))
    (Utilities|IsValid _playerpawn
      (:Is%20Valid
        (AthenaMovement|ServeronExitVehicle (LocalVariables|GetPlayerPawn) (Utilities|Struct|MakeVehicleExitData "ForceOnBlocking" true (Utilities|Struct|MakeVehicleExitLocationSelectionData "ByClosestToGivenLocation" (Transformation|GetActorLocation _playerpawn))))))
    (elif (GameplayTags|HasAnyTags (GameplayTags|GetOwnedGameplayTags _output_get :!class /Script/GameplayTags.GameplayTagAssetInterface) (GameplayTags|MakeLiteralGameplayTagContainer "(GameplayTags=((TagName=\"MovementMode.Parachuting\"),(TagName=\"MovementMode.Skydiving\"),(TagName=\"MovementMode.Skydiving.Following\")))") true)
      ;[0.02695 0.06967 0.151] End skydiving.
      (Parachute|EndSkydiving (LocalVariables|GetPlayerPawn))
      (elif (GameplayTags|HasAnyTags (GameplayTags|GetOwnedGameplayTags _output_get :!class /Script/GameplayTags.GameplayTagAssetInterface) (GameplayTags|MakeLiteralGameplayTagContainer "(GameplayTags=((TagName=\"MovementMode.Ziplining\"),(TagName=\"MovementMode.ZipliningOnSpline\")))") true)
        ;[0.05208 0.008721 0.05085] End ziplining.
        (EndMovement|EndMovementZiplining (LocalVariables|GetPlayerPawn))
        (elif (GameplayTags|HasAnyTags (GameplayTags|GetOwnedGameplayTags _output_get :!class /Script/GameplayTags.GameplayTagAssetInterface) (GameplayTags|MakeLiteralGameplayTagContainer "(GameplayTags=((TagName=\"MovementMode.Swinging\")))") true)
          ;[0.03744 0.1094 0.07412] End swinging.
          (Ability|SendGameplayEventtoActor (LocalVariables|GetPlayerPawn) "(TagName=\"Event.Athena.StopSwinging\")"))))))

(fn FindLauncher
  (params (DefaultLauncher :type Actor%20Object%20Reference))
  ;[0.06076 0.2344 0.225] The launcher is the hookshot instigator player pawn.
  (return (Variables|Transient|GetInstigatorPlayerPawn)))

(fn GetInitialKnockMag
  (params (Direction :type Vector :default "0, 0, 0") (OverrideXYMag :type Boolean) (XYMAgOverride :type Float%20%28double-precision%29))
  ;[0.08303 0.04865 0.1198] GetInitialKnockMag: Return the pull velocity for the target.
  (bind _velocity (Pull|CalculatePullVelocityforTarget (Variables|Transient|GetTargetCharacter)))
  (return _velocity))

(fn GetCustomMovementModeGameplayTags
  ;[0.04008 0.1351 0.1979] GetCustomMovementModeGameplayTags: Return the gameplay tags used to detect custom movement modes.
  (bind _keys (Utilities|Map|Keys (Variables|Hotfix|SameTeam|GetHotfixMap_PullSameTeamCustomMovement)))
  (LocalVariables|SetCustomMovementModeGameplayTags _keys)
  (bind _keys_2 (Utilities|Map|Keys (Variables|Hotfix|OtherTeam|GetHotfixMap_PullOtherTeamCustomMovement)))
  (for _array_element _keys_2
    (Utilities|Array|AddUnique (LocalVariables|GetCustomMovementModeGameplayTags) _array_element))
  (return (LocalVariables|GetCustomMovementModeGameplayTags)))

(fn GetInstigatorPlayerPawn
  ;[0.01955 0.1146 0.09622] GetInstigatorPlayerPawn: Return the hooking player pawn.
  ;;[0.1094 0.04899 0.1075] Return the instigator player pawn if it is already cached.
  (Variables|Transient|GetInstigatorPlayerPawn
    (:then
      (return (Variables|Transient|GetInstigatorPlayerPawn)))
    (:else
      (return (Utilities|Casting|CastToFortPlayerPawn (Variables|/Script/Engine.Actor|GetInstigator (Ability|EffectContext|GetEffectCauser (Ability|GetGrantedbyEffectContext))))))))

(fn GetMMEPullDuration
  (Utilities|FlowControl|Sequence
    (:then_0
      (bind _targetcharacter (Variables|Transient|GetTargetCharacter)
        (:then
          (bind _instigatorplayerpawn (Variables|Transient|GetInstigatorPlayerPawn)
            (:then
              (return (Math|Float|Clamp(Float) (Math|Float|SafeDivide (Math|Vector|Distance(Vector) (Transformation|GetActorLocation _instigatorplayerpawn) (Transformation|GetActorLocation _targetcharacter)) (ScalableFloat|GetValueatLevel (Variables|hotfix|GetHotfix_MMEPullSpeed))) 0.1 (Variables|Default|GetFailsafeDelay))))
            (:else)))
        (:else)))
    (:then_1
      (return 0.0))))

(fn GetTargetCharacterFromAvatar
  (params (AvatarActor :type Actor%20Object%20Reference))
  ;[0.05027 0.0111 0.1146] GetTargetCharacterFromAvatar: Returns the target character that is being hooked.
  (Utilities|FlowControl|Sequence
    (:then_0
      (Utilities|IsValid AvatarActor
        (:Is%20Valid
          ;;[0.03616 0.1719 0.05067] If target character is valid, return target character directly.
          (Utilities|IsValid (Variables|Transient|GetTargetCharacter)
            (:Is%20Valid
              (LocalVariables|SetRetCharacter (Variables|Transient|GetTargetCharacter)))
            (:Is%20Not%20Valid
              ;;[0.06141 0.1925 0.2188] Handle the avatar actor being the a player pawn.
              (bind _asfort_player_pawn (Utilities|Casting|CastToFortPlayerPawn AvatarActor)
                (:then
                  (LocalVariables|SetRetCharacter _asfort_player_pawn)
                  ;;;[0.02434 0.1354 0.1026] If the avatar is a DBNO player that is being hoisted, then the target player is the hoister.
                  (if (Health|IsDBNO _asfort_player_pawn)
                    (Utilities|IsValid (Class|FortPlayerPawn|GetDBNOHoister _asfort_player_pawn)
                      (:Is%20Valid
                        (LocalVariables|SetRetCharacter (Class|FortPlayerPawn|GetDBNOHoister _asfort_player_pawn))))))
                (:CastFailed
                  ;;[0.02098 0.04688 0.01206] Handle the avatar actor being an AI Pawn such as a wolf or boar.
                  (bind _asfort_aipawn (Utilities|Casting|CastToFortAIPawn AvatarActor))
                  (LocalVariables|SetRetCharacter _asfort_aipawn))))))))
    (:then_1
      ;;[0.1146 0.1128 0.02943] Return character
      (return (LocalVariables|GetRetCharacter)))))

(fn HandlePullFallingComplete
  (params (Character :type Character%20Object%20Reference) (PrevMovementMode :type EMovementMode%20Enum :default "MOVE_None") (PreviousCustomMode :type Byte))
  ;[0.03994 0.1467 0.1667] End the ability as soon as the pulled player changes their movement mode.
  (Ability|EndAbility))

(fn CanActivateAbility
  (params (ActorInfo :type Gameplay%20Ability%20Actor%20Info%20Structure) (Handle :type Gameplay%20Ability%20Spec%20Handle%20Structure))
  (bind (_owneractor _avataractor) (Utilities|Struct|BreakGameplayAbilityActorInfo ActorInfo))
  (bind _returnvalue (Ability|GetSourceObjectNonInstanced Handle ActorInfo))
  ;[0.0218 0.09896 0.06965] CanActivateAbility: Run checks here to ensure that the instigator pawn can actually hook the target actor.
  (Utilities|FlowControl|Sequence
    (:then_0
      ;;[0.08833 0.01668 0.125] Since CanActivateAbility is called before the ability is instanced, this alternate method is employed
      ;; instead of GetInstigatorPlayerPawn() function that is used everywhere else in the ability.
      (bind _asfort_weapon (Utilities|Casting|CastToFortWeapon _returnvalue)
        (:then
          (bind _asfort_player_pawn (Utilities|Casting|CastToFortPlayerPawn (Actor|GetOwner _asfort_weapon)))
          (bind _character (General|GetTargetCharacterfromAvatar _avataractor))
          (Utilities|IsValid _character
            (:Is%20Valid
              (bind _bsuccessful (General|CanInstigatorPullTarget _asfort_player_pawn _character))
              (if _bsuccessful
                (return 0 true)))))
        (:CastFailed
          ;;[0.08527 0.02651 0.1042] Consider AI Pawn using the ability.
          (bind _asfort_player_pawn_2 (Utilities|Casting|CastToFortPlayerPawn _returnvalue))
          (if (GameplayTags|HasMatchingGameplayTag _asfort_player_pawn_2 "(TagName=\"Pawn.Athena.NonParticipant\")")
            (bind _character_2 (General|GetTargetCharacterfromAvatar _avataractor))
            (Utilities|IsValid _character_2
              (:Is%20Valid
                (bind _bsuccessful_2 (General|CanInstigatorPullTarget _asfort_player_pawn_2 _character_2))
                (if _bsuccessful_2
                  (return 0 true))))))))
    (:then_1
      (return 0 false))))

(fn NotifyAnalyticsSuccessfulHookshot)

(fn PullAIPawn
  (params (AIPawn :type Fort%20AIPawn%20Object%20Reference))
  ;[0.1042 0.02098 0.1034] PullAIPawn: Pull the AI Pawn with the appropriate velocity.
  (Utilities|IsValid AIPawn
    (:Is%20Valid
      (bind _charactermovement (Variables|/Script/Engine.Character|GetCharacterMovement AIPawn)
        (:then
          (Pawn|Components|CharacterMovement|SetMovementMode _charactermovement "MOVE_Falling" 0) ; Set Falling in case we were Nav Mesh Walking.\nThat was stopping the Launch from working.
          (bind _velocity (Pull|CalculatePullVelocityforTarget AIPawn))
          (Character|LaunchCharacter AIPawn _velocity true true)
          (Pawn|Character|ResetFallingHeight AIPawn)
          (Variables|/Script/FortniteGame.FortPawn|SetbIgnoreNextFallingDamage true AIPawn))
        (:else)))))

(fn PullPlayerPawn
  (params (PlayerPawn :type Fort%20Player%20Pawn%20Object%20Reference))
  ;[0.01605 0.05814 0.0625] PullPlayerPawn: Pull the player pawn with the appropriate pull velocity.
  (Utilities|IsValid PlayerPawn
    (:Is%20Valid
      (bind _velocity (Pull|CalculatePullVelocityforTarget PlayerPawn))
      (Character|UnCrouch PlayerPawn)
      (bind _launchedpawn (Utilities|LaunchPlayerPawn (Variables|Transient|GetInstigatorPlayerPawn) PlayerPawn _velocity true true true true true true true))
      (Pawn|Character|ResetFallingHeight _launchedpawn))))

(fn PullTargetCharacter
  ;[0.2083 0.03017 0.1962] PullTargetCharacter: Based on the type of TargetCharacter, execute the appropriate pull function.
  (Utilities|FlowControl|Sequence
    (:then_0
      (if (not (ScalableFloat|GetValueasBool (Variables|hotfix|GetHotfix_UseMMEPull)))
        (bind _targetcharacter (Variables|Transient|GetTargetCharacter)
          (:then
            (bind _asfort_player_pawn (Utilities|Casting|CastToFortPlayerPawn _targetcharacter)
              (:then
                (Pull|PullPlayerPawn _asfort_player_pawn))
              (:CastFailed
                (bind _asfort_aipawn (Utilities|Casting|CastToFortAIPawn _targetcharacter))
                (Pull|PullAIPawn _asfort_aipawn)))
            (return true))
          (:else))))
    (:then_1
      (return false))))
