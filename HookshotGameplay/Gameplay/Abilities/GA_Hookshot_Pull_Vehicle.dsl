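; Hookshot pull ability on a vehicle, extending a knockback parent to pull and spin the hooked vehicle.
; Activates only when hotfix-enabled and the hookshot's player is neither riding nor standing on that vehicle.
; CalcVehicleMult adds a scaled share of current vehicle velocity to the parent's result.
; LaunchVehicle follows the parent launch with mass-scaled roll and yaw spin, derived from the angle between pull direction and impact offset.
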(event Ability|EventActivateAbility
  (|CallParentFunction) ; This node is disabled and will not be called.\nDrag off pins to build functionality.
)

(event Ability|EventOnEndAbility
  (params (bWasCancelled :type Boolean)))

(fn CalcVehicleMult
  (params (InVec :type Vector :default "0, 0, 0"))
  ;[0.01004 0.174 0.1927] CalcVehicleMult: Include the velocity of the vehicle to the calculation.
  (Utilities|FlowControl|Sequence
    (:then_0
      (bind _asfort_vehicle_interface (Utilities|Casting|CastToFortVehicleInterface (Variables|KnockbackParent|GetTarget)))
      (bind _returnvalue (Vehicle|GetVehicleActor _asfort_vehicle_interface :!class /Script/FortniteGame.FortVehicleInterface))
      (Utilities|IsValid _returnvalue
        (:Is%20Valid
          (return (+ (* (Transformation|GetVelocity _returnvalue) (ScalableFloat|GetValueatLevel (Variables|Hotfix|GetHotfix_VehiclePullVelocityMaintainMultiplier))) (|CallParentFunction InVec))))))
    (:then_1
      (return (|CallParentFunction InVec)))))

(fn CanActivateAbility
  (params (ActorInfo :type Gameplay%20Ability%20Actor%20Info%20Structure) (Handle :type Gameplay%20Ability%20Spec%20Handle%20Structure))
  ;[0.06677 0.1688 0.1875] CanActivateAbility: Only activate if the owner is a vehicle and hotfix allows it.
  (Utilities|FlowControl|Sequence
    (:then_0
      (bind _asfort_vehicle_interface (Utilities|Casting|CastToFortVehicleInterface (Utilities|Struct|BreakGameplayAbilityActorInfo ActorInfo)))
      (bind _returnvalue (Vehicle|GetVehicleActor _asfort_vehicle_interface :!class /Script/FortniteGame.FortVehicleInterface))
      (Utilities|IsValid _returnvalue
        (:Is%20Valid
          (if (ScalableFloat|GetValueasBool (Variables|Hotfix|GetHotfix_VehiclePullEnable))
            (bind _asfort_weapon (Utilities|Casting|CastToFortWeapon (Ability|GetSourceObjectNonInstanced Handle ActorInfo)))
            (bind _asfort_player_pawn (Utilities|Casting|CastToFortPlayerPawn (Actor|GetOwner _asfort_weapon)))
            (bind _charactermovement (Variables|/Script/Engine.Character|GetCharacterMovement _asfort_player_pawn)
              (:then
                (bind _asprimitive_component (Utilities|Casting|CastToPrimitiveComponent (Pawn|Components|CharacterMovement|GetMovementBaseObject _charactermovement))
                  (:then
                    (if (!= (Components|GetOwner _asprimitive_component) _returnvalue)
                      _))
                  (:CastFailed
                    _))
                (if (!= (Vehicle|GetVehicleActor _asfort_player_pawn :!class /Script/FortniteGame.FortPlayerPawn) _returnvalue)
                  (return 0 true)))
              (:else))))))
    (:then_1
      (return 0 false))))

(fn LaunchVehicle
  (params (VehicleInterface :type Fort%20Vehicle%20Interface%20Interface) (LaunchVelocity :type Vector%20%28by%20ref%29 :default "0, 0, 0"))
  (bind (_bblockinghit _binitialoverlap _time _distance _location _impactpoint) (Collision|BreakHitResult (Ability|EffectContext|GetHitResult (Ability|GetGrantedbyEffectContext))))
  ;[0.03892 0.1925 0.2188] LaunchVehicle: Overridden to also add angular rotation to the vehicle based on the pull direction and the hookshot impact location.
  (|CallParentFunction VehicleInterface LaunchVelocity)
  (Utilities|FlowControl|Sequence
    (:then_0
      (bind _returnvalue (Vehicle|GetVehicleActor VehicleInterface :!class /Script/FortniteGame.FortVehicleInterface))
      (Utilities|IsValid _returnvalue
        (:Is%20Valid
          (bind _returnvalue_2 (FortPhysicsSimulationLibrary|GetMass _returnvalue))
          (Audio|BroadcastAbilitySpinThrustatLocation 0 _returnvalue (Math|Vector|MakeVector (Math|Float|SafeDivide (* (ScalableFloat|GetValueatLevel (Variables|Hotfix|GetHotfix_VehiclePullRollMagnitude)) (Utilities|Operators|Multiply (Math|Float|Sign(Float) (.z (Math|Vector|CrossProduct (Math|Vector|Normalize LaunchVelocity) (Math|Vector|Normalize (- _impactpoint (Transformation|GetActorLocation _returnvalue)))))) (Math|Float|MapRangeClamped (Math|Float|ClampAngle (Math|Trig|Acos(Degrees) (Math|Vector|DotProduct (Math|Vector|Normalize (- _impactpoint (Transformation|GetActorLocation _returnvalue))) (Math|Vector|Normalize LaunchVelocity))) 0.0 9e+01) 0.0 9e+01 0.0 1.0) -1.0)) (Math|Float|SafeDivide _returnvalue_2 1e+02)) 0.0 (Math|Float|SafeDivide (* (ScalableFloat|GetValueatLevel (Variables|Hotfix|GetHotfix_VehiclePullYawMagnitude)) (Utilities|Operators|Multiply (Math|Float|Sign(Float) (.z (Math|Vector|CrossProduct (Math|Vector|Normalize LaunchVelocity) (Math|Vector|Normalize (- _impactpoint (Transformation|GetActorLocation _returnvalue)))))) (Math|Float|MapRangeClamped (Math|Float|ClampAngle (Math|Trig|Acos(Degrees) (Math|Vector|DotProduct (Math|Vector|Normalize (- _impactpoint (Transformation|GetActorLocation _returnvalue))) (Math|Vector|Normalize LaunchVelocity))) 0.0 9e+01) 0.0 9e+01 0.0 1.0) -1.0)) (Math|Float|SafeDivide _returnvalue_2 1e+02))) 7.8e+02 "None"))))
    (:then_1
      (return))))
